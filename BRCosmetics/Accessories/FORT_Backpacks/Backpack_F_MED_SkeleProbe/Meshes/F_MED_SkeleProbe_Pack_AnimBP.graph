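# Animation Blueprint for the SkeleProbe backpack cosmetic that keeps the pack fitted to the wearer's back and adds secondary motion to its rope, star and zipper.
# Each update maps the owner's spine_05-to-spine_03 yaw delta from ref pose into additive pack translation and roll, atop a fixed offset.
# Two chained sets of Anim Dynamics then simulate the three dangling bones with separate tuning, blended on or off by whether the player is skydive-floating.
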AnimBlueprint /BRCosmetics/Accessories/FORT_Backpacks/Backpack_F_MED_SkeleProbe/Meshes/F_MED_SkeleProbe_Pack_AnimBP
target_skeleton=&/F_MED_SkeleProbe_Pack_Skeleton
graphs [
 AnimGraph {nodes [
   n1 = Get bIsSkydiveFloating()
   n2 = "Transform (Modify) Bone - Bone: pack"(ComponentPose=Local To Component(LocalPose=Input Pose()), Translation="0,0.6,2", Rotation="0,0,0") {
    translationMode=BMM_Additive
    translationSpace=BCS_BoneSpace
    alphaBoolBlend {blendOption=Linear}
    boneToModify {boneName=pack}
    translation {y=0.6, z=2}
   }
   n3 = "Transform (Modify) Bone - Bone: pack"(ComponentPose=@n2, Translation=Get backpack_translation(), Rotation=Get backpack_rotate()) {
    rotationMode=BMM_Additive
    scaleMode=BMM_Additive
    scaleSpace=BCS_BoneSpace
    translationMode=BMM_Additive
    alphaBoolBlend {blendOption=Linear}
    boneToModify {boneName=pack}
   }
   n4 = Get not_floating() -> (not_floating:bool)
   n5 = Anim Dynamics - Bone: dyn_Rope!(ComponentPose=@n3, bAlphaBoolEnabled:bool=@n4, GravityOverride:Vector="0,0,-980", ExternalForce="0,0,0") -> (Pose) {
    alphaInputType=Bool
    angularBiasOverride=1
    angularDampingOverride=1
    angularSpringConstant=1000
    bAngularSpring=true
    bOverrideAngularBias=true
    bOverrideAngularDamping=true
    bUseGravityOverride=true
    bUsePlanarLimit=false
    simulationSpace=RootRelative
    alphaBoolBlend {blendInTime=0.1, blendOption=CircularInOut, blendOutTime=0.1}
    boundBone {boneName=dyn_Rope}
    componentAppliedLinearAccClamp=<3000 3000 3000>
    componentLinearAccScale=<0.3 0.3 0.3>
    componentLinearVelScale=<0.4 0.4 0.4>
    physicsBodyDefinitions [{
      collisionType=CustomSphere
      sphereCollisionRadius=2.5
      boundBone {boneName=dyn_Rope}
      boxExtents=<8 10 15>
      constraintSetup {angularLimitsMax=<0 35 0>, angularLimitsMin=<0 -35 -60>, angularTarget=<0 0 0>}
      localJointOffset=<3 0 0>
     }]
   }
   n6 = Anim Dynamics - Bone: dyn_Star!(ComponentPose=@n5, bAlphaBoolEnabled:bool=@n4, GravityOverride:Vector="0,0,-900", ExternalForce="0,0,0") -> (Pose) {
    alphaInputType=Bool
    angularBiasOverride=1
    angularDampingOverride=1
    angularSpringConstant=1000
    bAngularSpring=true
    bOverrideAngularBias=true
    bOverrideAngularDamping=true
    bUseGravityOverride=true
    bUsePlanarLimit=false
    simulationSpace=RootRelative
    alphaBoolBlend {blendInTime=0.2, blendOption=CircularIn, blendOutTime=0.3}
    boundBone {boneName=dyn_Star}
    componentAppliedLinearAccClamp=<3000 3000 3000>
    componentLinearAccScale=<0.3 0.3 0.3>
    componentLinearVelScale=<0.5 0.5 0.5>
    physicsBodyDefinitions [{
      collisionType=CustomSphere
      sphereCollisionRadius=2.5
      boundBone {boneName=dyn_Star}
      boxExtents=<8 6 6>
      constraintSetup {angularLimitsMax=<0 20 0>, angularLimitsMin=<0 -20 -30>, angularTarget=<0 0 0>}
      localJointOffset=<4 0 0>
     }]
   }
   n7 = Anim Dynamics - Bone: dyn_Zipper!(ComponentPose=@n6, bAlphaBoolEnabled:bool=@n4, GravityOverride:Vector="0,0,-300", ExternalForce="0,0,0") -> (Pose) {
    alphaInputType=Bool
    angularBiasOverride=1
    angularDampingOverride=0.3
    angularSpringConstant=1000
    bAngularSpring=true
    bOverrideAngularDamping=true
    bUseGravityOverride=true
    bUsePlanarLimit=false
    simulationSpace=RootRelative
    alphaBoolBlend {blendInTime=0.5, blendOption=Linear, blendOutTime=0.7}
    boundBone {boneName=dyn_Zipper}
    componentAppliedLinearAccClamp=<200 200 200>
    componentLinearAccScale=<0.3 0.3 0.3>
    componentLinearVelScale=<0.4 0.4 0.4>
    physicsBodyDefinitions [{
      collisionType=CustomSphere
      sphereCollisionRadius=2.5
      boundBone {boneName=dyn_Zipper}
      boxExtents=<4 4 4>
      constraintSetup {angularLimitsMax=<0 130 10>, angularLimitsMin=<0 1 -10>, angularTarget=<0 0 0>}
      localJointOffset=<2.5 0 0>
     }]
   }
   n8 = Anim Dynamics - Bone: dyn_Rope(ComponentPose=@n7, bAlphaBoolEnabled=@n1, GravityOverride="0,0,-800", ExternalForce="0,0,0") {
    alphaInputType=Bool
    angularBiasOverride=1
    angularDampingOverride=1
    angularSpringConstant=1000
    bAngularSpring=true
    bOverrideAngularBias=true
    bOverrideAngularDamping=true
    bUseGravityOverride=true
    bUsePlanarLimit=false
    simulationSpace=RootRelative
    alphaBoolBlend {blendInTime=0.1, blendOption=Linear, blendOutTime=0.1}
    boundBone {boneName=dyn_Rope}
    componentAppliedLinearAccClamp=<200 200 200>
    componentLinearAccScale=<0.1 0.1 0.1>
    componentLinearVelScale=<0.3 0.3 0.3>
    physicsBodyDefinitions [{
      collisionType=CustomSphere
      sphereCollisionRadius=2.5
      boundBone {boneName=dyn_Rope}
      boxExtents=<8 10 15>
      constraintSetup {angularLimitsMax=<0 20 0>, angularLimitsMin=<0 -20 -60>, angularTarget=<0 0 0>}
      localJointOffset=<4 0 0>
     }]
   }
   n9 = Anim Dynamics - Bone: dyn_Star(ComponentPose=@n8, bAlphaBoolEnabled=@n1, GravityOverride="0,0,-600", ExternalForce="0,0,0") {
    alphaInputType=Bool
    angularBiasOverride=1
    angularDampingOverride=1
    angularSpringConstant=1000
    bAngularSpring=true
    bOverrideAngularDamping=true
    bUseGravityOverride=true
    bUsePlanarLimit=false
    simulationSpace=RootRelative
    alphaBoolBlend {blendInTime=0.2, blendOption=Linear, blendOutTime=0.4}
    boundBone {boneName=dyn_Star}
    componentAppliedLinearAccClamp=<3000 3000 3000>
    componentLinearAccScale=<0.4 0.4 0.4>
    componentLinearVelScale=<0.6 0.6 0.6>
    physicsBodyDefinitions [{
      collisionType=CustomSphere
      sphereCollisionRadius=2.5
      boundBone {boneName=dyn_Star}
      boxExtents=<12 10 12>
      constraintSetup {angularLimitsMax=<0 20 0>, angularLimitsMin=<0 -20 -30>, angularTarget=<0 0 0>}
      localJointOffset=<4 0 0>
     }]
   }
   n10 = Anim Dynamics - Bone: dyn_Zipper(ComponentPose=@n9, bAlphaBoolEnabled=@n1, GravityOverride="0,0,-300", ExternalForce="0,0,0") {
    alphaInputType=Bool
    angularBiasOverride=1
    angularDampingOverride=0.99
    angularSpringConstant=1000
    bAngularSpring=true
    bOverrideAngularDamping=true
    bUseGravityOverride=true
    bUsePlanarLimit=false
    simulationSpace=RootRelative
    alphaBoolBlend {blendInTime=0.1, blendOption=Linear, blendOutTime=0.1}
    boundBone {boneName=dyn_Zipper}
    componentAppliedLinearAccClamp=<200 200 200>
    componentLinearAccScale=<0.3 0.3 0.3>
    componentLinearVelScale=<0.4 0.4 0.4>
    physicsBodyDefinitions [{
      collisionType=CustomSphere
      sphereCollisionRadius=2.5
      boundBone {boneName=dyn_Zipper}
      boxExtents=<4 4 4>
      constraintSetup {angularLimitsMax=<0 100 10>, angularLimitsMin=<0 1 -10>, angularTarget=<0 0 0>}
      localJointOffset=<2.5 0 0>
     }]
   }
   Output Pose(Result=Component To Local(ComponentPose=@n10)) {layerGroup=None}
  ]}
 EventGraph {nodes [
   Event BlueprintUpdateAnimation()
   > n1 = Sequence()
   n2 = GetOwnerSkeletalMeshComponent()
   n3 = BreakRotator(InRot=BreakTransform(InTransform=GetDeltaTransformFromRefPose(self=@n2, BoneName=spine_05, BaseName=spine_03)).Rotation)
   > Set backpack_rotate(backpack_rotate=MakeRotator(Roll=MapRangeClamped(Value=@n3.Yaw, InRangeA=0, InRangeB=-30, OutRangeA=5, OutRangeB=25)))
   > Set backpack_translation(backpack_translation=MakeVector(Y=MapRangeClamped(Value=@n3.Yaw, InRangeA=0, InRangeB=-30, OutRangeB=-5)))
   Set not_floating(execute=@n1.then_1, not_floating:bool=NOT Boolean(A=Get bIsSkydiveFloating())) -> (Output_Get:bool)
   TryGetPawnOwner()
  ]}
]
variables [
 backpack_rotate {type=Rotator}
 backpack_translation {type=Vector}
 Gravity Override {type=Vector}
 Gravity Override Params {type=GravityOverrideParamsStruct}
 not_floating {type=boolean}
]
